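# Niagara vehicle boost cosmetic that trails a red laser ribbon and layered grey smoke behind the car.
# Lasers002 draws a double-rendered ribbon; three sprite emitters (middle, end, wheel) spawn flipbook smoke at skeletal mesh sockets, all offset by DTCarOffset.
# Laser and end smoke inherit source movement with curve-driven falloff over life.
# Middle smoke switches velocity range for frontend preview; painted-variant override modules are disabled in all emitters.
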NiagaraSystem /VehicleCosmetics/Boosts/Boost_LaserSmoke/FX/NS_Boost_LaserSmokeB
dependencies {
 usedDataInterfaces [
  &NiagaraDataInterfaceSpriteRendererInfo
  &NiagaraDataInterfaceCurve
  &NiagaraDataInterfaceVector2DCurve
  &NiagaraDataInterfaceVectorCurve
  &NiagaraDataInterfaceSkeletalMesh
 ]
 usedDynamicInputs [&Vector2DFromFloat]
 usedModules [
  &self:SystemSpawnScript.SystemScriptSource.SystemScriptGraph.NiagaraNodeAssignment_0.SetVariables_#1
  &SystemState
 ]
}
emitters {
 Lasers002 {
  emitterUpdateScript [
   &EmitterState()
   &VC_RateQualityHandler(High Spawn Rate=60, Low Spawn Rate=60, Medium Spawn Rate=60, Use Custom Rate?=true)
   &VC_SpawnRibbon_v1(SpawnRate=60)
  ]
  particleSpawnScript [
   &VC_PaintedVariantsOverride() {enabled=false}
   &VC_InitializeRibbon_v1(
    Color=<rgba 1.5 0 0 1>
    Custom Offset=true
    Enable Skeletal Position=true
    Lifetime=1
    Local Vector Offset=&AddVectorToPosition(Position=$User.DTCarOffset, Vector=<-60 0 -10>)
    Offset Amount=1
    Ribbon Width=50
   )
   &AddVelocity(Velocity=<-800 0 0>) {module=AddVelocity001}
  ]
  particleUpdateScript [
   &ParticleState()
   &InheritSourceMovement(
    Applied Position Scale=&VectorFromCurve(VectorCurve=DataInterface() {propertyValues {
       XCurve {keys [<at 0 1 RCIM_Cubic>, {interpMode=RCIM_Cubic, leaveTangent=-1.655, time=0.7, value=0}]}
       YCurve {keys [<at 0 1 RCIM_Cubic>, {interpMode=RCIM_Cubic, leaveTangent=-1.655, time=0.7, value=0}]}
       ZCurve {keys [<at 0 1 RCIM_Cubic>, {interpMode=RCIM_Cubic, leaveTangent=-1.655, time=0.7, value=0}]}
      }})
    ApplyForce=false
    ApplyPosition=true
    ApplyPositionFalloff=false
   )
   &SolveForcesAndVelocity(Clamp Velocity=false, Limit Acceleration=false)
   &ScaleRibbonWidth(Ribbon Width Scale=&FloatFromCurve(FloatCurve=DataInterface() {propertyValues {Curve {keys [
         <at 0 1>
         {arriveTangent=-1, interpMode=RCIM_Cubic, leaveTangent=-1, time=0.7033, value=0.6871}
         {interpMode=RCIM_Cubic, tangentMode=RCTM_User, time=1, value=0}
        ]}}}))
   &VC_UpdateRibbon_v1(Enable Skeletal Position=true)
   &ScaleColor(Scale Alpha=2, Scale RGB=&VectorFromFloat(Value=3), ScaleA=true, ScaleRGB=true)
  ]
  rendererClasses [&NiagaraRibbonRendererProperties]
  renderers [
   {rendererClass=&NiagaraRibbonRendererProperties}
   {rendererClass=&NiagaraRibbonRendererProperties, rendererIndex=1}
  ]
 }
 SmokeEnd001 {
  emitterUpdateScript [
   &EmitterState()
   &SpawnRate(SpawnRate=200, Use Spawn Probability=false)
  ]
  particleSpawnScript [
   &VC_PaintedVariantsOverride() {enabled=false}
   &InitializeParticle(Lifetime=0.8, Sprite Size Mode=::NewEnumerator3, Uniform Sprite Size=130, UsePositionOffset=false, Write Material Random=true)
   &PartitionParticles(Number Of Partitions=1)
   &SkeletalMeshLocation(
    Bone / Socket Index=$Particles.PartitionParticles.Partition
    Bone / Socket Sampling Mode=::NewEnumerator5
    Interpolate Mesh Sampling=false
    MaskSpawnGroup=false
    Mesh Sampling Type=::NewEnumerator6
    Sampled Position Offset=&AddVectorToPosition(Position=$User.DTCarOffset, Vector=<-30 0 -10>)
    Skeletal Mesh=$System.SkeletalMesh
   )
   &AddVelocity(Velocity=&RandomRangeVector(Minimum=<-400 0 -20>, Maximum=<-600 1 -20>))
   &Color(Color=<rgba 0.4219 0.4219 0.4219 0.0435>)
  ]
  particleUpdateScript [
   &ParticleState()
   &SubUVAnimation(
    Random Start Frame=true
    "Sprite Renderer "=DataInterface() {propertyValues {SpriteRenderer=&self:SmokeEnd001_0.NiagaraSpriteRendererProperties_0}}
    UseEndFrame=false
    UseStartFrame=false
   )
   &DynamicMaterialParameters(
    Index 0 Param 1=2
    Index 0 Param 2=&FloatFromCurve(FloatCurve=DataInterface() {propertyValues {Curve {keys [<at 0 0.5 RCIM_Cubic>,<at 1 1 RCIM_Cubic>]}}})
    Index 0 Param 3=&RandomRangeFloat(Minimum=0.1, Maximum=0.25)
    Index 0 Param 4=1
    Module.Index0Param0Write=true
    Module.Index0Param1Write=true
    Module.Index0Param2Write=true
    Module.Index0Param3Write=true
   )
   &InheritSourceMovement(
    Applied Position Scale=&VectorFromFloat(Value=&FloatFromCurve(FloatCurve=DataInterface() {propertyValues {Curve {keys [
          {interpMode=RCIM_Cubic, tangentMode=RCTM_Break, time=0, value=1}
          {arriveTangent=-2.9025, interpMode=RCIM_Cubic, tangentMode=RCTM_Break, time=1, value=0}
         ]}}}))
    ApplyForce=false
    ApplyPosition=true
    ApplyPositionFalloff=false
   )
   &SpriteFacingAndAlignment(
    Sprite Alignment=&Add_Vector(
     A=&NormalizeVector(Vector To Normalize=$Particles.Velocity, Safe Normalize=true, Safe Fallback Threshold=0, Safe Fallback Vector=<0 0 0.01>)
     B=&MakeVector(
      X=0
      Y=&Multiply_Float(
       A=&RandomRangeFloat(Minimum=-1, Maximum=1)
       B=&FloatFromCurve(FloatCurve=DataInterface() {propertyValues {Curve {keys [<at 0 0.05>,<at 0.1 0>]}}})
      )
      Z=&Multiply_Float(
       A=&RandomRangeFloat(Minimum=-1, Maximum=1)
       B=&FloatFromCurve(FloatCurve=DataInterface() {propertyValues {Curve {keys [<at 0 1 RCIM_Cubic>,<at 1 0 RCIM_Cubic>]}}})
      )
     )
    )
    WriteAlignment=true
    WriteFacing=false
   ) {enabled=false, module=SpriteFacingAndAlignment001}
   &SolveForcesAndVelocity(Clamp Velocity=false, Limit Acceleration=false)
   &ScaleSpriteSize(
    Non-Uniform Curve Sprite Scale=DataInterface() {propertyValues {
      XCurve {keys [
        {arriveTangent=1.5105, interpMode=RCIM_Cubic, leaveTangent=1.5105, tangentMode=RCTM_User, time=0.4937, value=0}
        {arriveTangent=-1.0633, interpMode=RCIM_Cubic, leaveTangent=-1.0633, tangentMode=RCTM_User, time=0.75, value=0.3}
        <at 1 0.2 RCIM_Cubic>
       ]}
      YCurve {keys [{tangentMode=RCTM_User, time=0.4913, value=0}, {tangentMode=RCTM_User, time=0.75, value=1.2}]}
     }}
    Scale Sprite Size Mode=::NewEnumerator3
   )
   &ScaleColor(
    Scale Alpha=&FloatFromCurve(FloatCurve=DataInterface() {propertyValues {Curve {
        defaultValue=0.519
        keys [
         {interpMode=RCIM_Cubic, tangentMode=RCTM_User, time=0.0461, value=0}
         {arriveTangent=-5.212, interpMode=RCIM_Cubic, leaveTangent=-5.212, tangentMode=RCTM_User, time=0.5, value=2.4706}
         {arriveTangent=-5.4079, tangentMode=RCTM_User, time=1, value=0}
        ]
       }}})
    ScaleA=true
    ScaleRGB=true
   )
  ]
  rendererClasses [&NiagaraSpriteRendererProperties]
  renderers [{rendererClass=&NiagaraSpriteRendererProperties}]
 }
 SmokeMiddle001 {
  emitterUpdateScript [
   &EmitterState()
   &SpawnRate(SpawnRate=250, Use Spawn Probability=false)
  ]
  particleSpawnScript [
   &VC_PaintedVariantsOverride() {enabled=false}
   &InitializeParticle(
    Color=<rgba 1 1 1 1000>
    Lifetime=1.3
    Sprite Rotation Mode=::NewEnumerator2
    Sprite Size Mode=::NewEnumerator3
    Uniform Sprite Size=&RandomRangeFloat(Minimum=50, Maximum=70)
    UsePositionOffset=false
    Write Material Random=true
   )
   &PartitionParticles(Number Of Partitions=1)
   &SkeletalMeshLocation(
    Bone / Socket Index=$Particles.PartitionParticles.Partition
    Bone / Socket Sampling Mode=::NewEnumerator5
    Interpolate Mesh Sampling=false
    MaskSpawnGroup=false
    Mesh Sampling Type=::NewEnumerator6
    Sampled Position Offset=&AddVectorToPosition(Position=$User.DTCarOffset, Vector=<-35 0 -15>)
    Skeletal Mesh=$System.SkeletalMesh
   )
   &AddVelocity(Velocity=&MakeCustomVectorFromBool(
     Bool=$User.bIsFrontendPreview
     True Vector=&RandomRangeVector(Minimum=<-400 0 0>, Maximum=<-500 0 0>)
     False Vector=&RandomRangeVector(Minimum=<-400 0 0>, Maximum=<-500 1 1>)
    ))
   &Color(Color=<rgba 0.2604 0.2604 0.2604 1>)
  ]
  particleUpdateScript [
   &ParticleState()
   &SubUVAnimation(
    "Sprite Renderer "=DataInterface() {propertyValues {SpriteRenderer=&self:SmokeMiddle001_0.NiagaraSpriteRendererProperties_0}}
    UseEndFrame=false
    UseStartFrame=false
   )
   &ScaleSpriteSize(Uniform Curve Sprite Scale=DataInterface() {propertyValues {Curve {defaultValue=0, keys [{tangentMode=RCTM_User, time=0, value=1.3}, {tangentMode=RCTM_User, time=0.8, value=0.5}]}}})
   &SolveForcesAndVelocity(Clamp Velocity=false, Limit Acceleration=false)
   &ScaleColor(
    Scale Alpha=&RandomRangeFloat(
     Minimum=&FloatFromCurve(FloatCurve=DataInterface() {propertyValues {Curve {
         defaultValue=1
         keys [
          <at 0 0.1 RCIM_Cubic>
          {arriveTangent=0.2949, interpMode=RCIM_Cubic, leaveTangent=0.2949, time=0.4323, value=0.3159}
          {arriveTangent=-0.2254, interpMode=RCIM_Cubic, leaveTangent=-0.2254, time=0.7323, value=0.3159}
          <at 1 0.1879 RCIM_Cubic>
         ]
        }}})
     Maximum=&FloatFromCurve(FloatCurve=DataInterface() {propertyValues {Curve {
         defaultValue=1
         keys [
          <at 0 0.1 RCIM_Cubic>
          {arriveTangent=0.9966, interpMode=RCIM_Cubic, leaveTangent=0.9966, time=0.4818, value=0.7743}
          {arriveTangent=-0.5251, interpMode=RCIM_Cubic, leaveTangent=-0.5251, time=0.6978, value=0.7954}
          <at 1 0.5022 RCIM_Cubic>
         ]
        }}})
    )
    ScaleA=true
    ScaleRGB=true
   )
   &DynamicMaterialParameters(
    Index 0 Param 1=2
    Index 0 Param 2=&FloatFromCurve(FloatCurve=DataInterface() {propertyValues {Curve {keys [<at 0 0.5 RCIM_Cubic>,<at 1 1 RCIM_Cubic>]}}})
    Index 0 Param 3=&RandomRangeFloat(Minimum=0.1, Maximum=0.25)
    Index 0 Param 4=1
    Module.Index0Float0Write=true
    Module.Index0Float1Write=true
    Module.Index0Float2Write=true
    Module.Index0Float3Write=true
   )
  ]
  rendererClasses [&NiagaraSpriteRendererProperties]
  renderers [{rendererClass=&NiagaraSpriteRendererProperties}]
 }
 SmokeWheel001 {
  emitterUpdateScript [
   &EmitterState()
   &SpawnRate(SpawnRate=30, Use Spawn Probability=false)
  ]
  particleSpawnScript [
   &VC_PaintedVariantsOverride() {enabled=false}
   &InitializeParticle(
    Color=<rgba 1 1 1 1000>
    Lifetime Max=0.8
    Lifetime Min=0.8
    Lifetime Mode=::NewEnumerator1
    Sprite Rotation Mode=::NewEnumerator2
    Sprite Size Mode=::NewEnumerator3
    Uniform Sprite Size=120
    UsePositionOffset=false
    Write Material Random=true
   )
   &PartitionParticles(Number Of Partitions=1)
   &SkeletalMeshLocation(
    Bone / Socket Index=$Particles.PartitionParticles.Partition
    Bone / Socket Sampling Mode=::NewEnumerator5
    Interpolate Mesh Sampling=false
    MaskSpawnGroup=false
    Mesh Sampling Type=::NewEnumerator6
    Sampled Position Offset=&AddVectorToPosition(Position=$User.DTCarOffset, Vector=<-45 0 -15>)
    Skeletal Mesh=$System.SkeletalMesh
   )
   &AddVelocity(Velocity=&MaskVectorBySpawnGroup(Masked Vector=<-200 0 -10>, Passthrough Vector=<-300 1 -10>))
   &Color(Color=<rgba 0.1719 0.1719 0.1719 1>)
  ]
  particleUpdateScript [
   &ParticleState()
   &SubUVAnimation(
    "Sprite Renderer "=DataInterface() {propertyValues {SpriteRenderer=&self:SmokeWheel001_0.NiagaraSpriteRendererProperties_0}}
    UseEndFrame=false
    UseStartFrame=false
   )
   &ScaleSpriteSize(Uniform Curve Sprite Scale=DataInterface() {propertyValues {Curve {defaultValue=0, keys [{tangentMode=RCTM_User, time=0, value=1.3}, {tangentMode=RCTM_User, time=0.8, value=0}]}}})
   &SolveForcesAndVelocity(Clamp Velocity=false, Limit Acceleration=false)
   &ScaleColor(
    Scale Alpha=&FloatFromCurve(FloatCurve=DataInterface() {propertyValues {Curve {keys [<at 0 0.6972 RCIM_Cubic>, {arriveTangent=-0.6972, interpMode=RCIM_Cubic, leaveTangent=-0.6972, time=0.2, value=0.6972}, <at 1 0 RCIM_Cubic>]}}})
    ScaleA=true
    ScaleRGB=true
   )
   &DynamicMaterialParameters(
    Index 0 Param 1=2
    Index 0 Param 2=&FloatFromCurve(FloatCurve=DataInterface() {propertyValues {Curve {keys [<at 0 0.5 RCIM_Cubic>,<at 1 1 RCIM_Cubic>]}}})
    Index 0 Param 3=&RandomRangeFloat(Minimum=0.1, Maximum=0.25)
    Index 0 Param 4=1
    Module.Index0Float0Write=true
    Module.Index0Float1Write=true
    Module.Index0Float2Write=true
    Module.Index0Float3Write=true
   )
  ]
  rendererClasses [&NiagaraSpriteRendererProperties]
  renderers [{rendererClass=&NiagaraSpriteRendererProperties}]
 }
}
summary {
 systemName=NS_Boost_LaserSmokeB
 emitters [
  {emitterName=SmokeMiddle001, rendererClasses [&NiagaraSpriteRendererProperties]}
  {emitterName=SmokeEnd001, rendererClasses [&NiagaraSpriteRendererProperties]}
  {emitterName=SmokeWheel001, rendererClasses [&NiagaraSpriteRendererProperties]}
  {emitterName=Lasers002, rendererClasses [&NiagaraRibbonRendererProperties]}
 ]
 userVariables [
  User.bIsFrontendPreview {type=NiagaraBool}
  User.bIsPainted {type=NiagaraBool}
  User.ColorVariantID {type=NiagaraInt32}
  User.DTCarOffset {type=NiagaraPosition, description="", defaultValue {}}
  User.intensity_trail {type=NiagaraFloat, defaultValue=3, description=""}
  User.Lines_ribbon {type=NiagaraFloat, defaultValue=5, description=""}
  User.PaintedColor {type=LinearColor, description="", defaultValue {struct=LinearColor, value=<rgba 1 1 1 1>}}
  User.PaintedColor2 {type=LinearColor, description="", defaultValue {struct=LinearColor, value=<rgba 1 1 1 1>}}
  User.PaintedScalar {type=NiagaraFloat, defaultValue=1, description=""}
  User.PortCount {type=NiagaraInt32, defaultValue=1}
 ]
}
